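# Animation Blueprint for the male medium Rocker Punk commando body, copying the attached parent mesh's pose and adding secondary motion and correctives.
# Weighted constraints and a bone-driven controller position zippers and lapels, while gravity-driven Anim Dynamics swings zippers and right-side straps.
# RBF Pose Drivers drive collar bones from neck and head, and corrective curves from the right upper arm.
# A remapped pistol-target-to-forearm distance, updated each frame, blends a right clavicle-end offset.
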AnimBlueprint /Game/Characters/Player/Male/Medium/Bodies/M_Med_Commando_RockerPunk/Meshes/M_MED_Rocker_Punk_Body_Skeleton_AnimBlueprint
target_skeleton=&/M_MED_Rocker_Punk_Skeleton
graphs [
 AnimGraph {nodes [
   n1 = "Transform (Modify) Bone - Bone: dyn_armpitScale"!(ComponentPose=Local To Component(LocalPose=Copy Pose From Mesh() {bUseAttachedParent=true}), Alpha=1, Rotation="0,0,0") -> (Pose) {
    scaleMode=BMM_Replace
    scaleSpace=BCS_BoneSpace
    translationMode=BMM_Additive
    translationSpace=BCS_BoneSpace
    alphaBoolBlend {blendOption=Linear}
    boneToModify {boneName=dyn_armpitScale}
    scale=<2 2 2>
    translation {x=-10}
   }
   n2 = "Constraint - const_zipperLegA_1 (PP)"!(ComponentPose=@n1, Alpha=1, ConstraintWeights_0=0.085, ConstraintWeights_1:float=0.915) -> (Pose) {
    alphaBoolBlend {blendOption=Linear}
    boneToModify {boneName=const_zipperLegA_1}
    constraintSetup [{targetBone {boneName=pelvis}}, {targetBone {boneName=thigh_twist_01_r}}]
    constraintWeights [0.085,0.915]
   }
   n3 = "Constraint - const_zipperLegB_1 (PPP)"!(ComponentPose=@n2, Alpha=1, ConstraintWeights_0=0.023, ConstraintWeights_1:float=0.943, ConstraintWeights_2:float=0.034) -> (Pose) {
    alphaBoolBlend {blendOption=Linear}
    boneToModify {boneName=const_zipperLegB_1}
    constraintSetup [{targetBone {boneName=pelvis}}, {targetBone {boneName=thigh_twist_01_r}}, {targetBone {boneName=thigh_r}}]
    constraintWeights [0.085,0.915,1]
   }
   n4 = Anim Dynamics - Bone: dyn_zipperLegA_2(ComponentPose=@n3, Alpha=1, ExternalForce="0,0,0") {
    bGravityOverrideInSimSpace=true
    bUseGravityOverride=true
    lODThreshold=1
    numSolverIterationsPostUpdate=4
    numSolverIterationsPreUpdate=16
    simulationSpace=RootRelative
    alphaBoolBlend {blendOption=Linear}
    boundBone {boneName=dyn_zipperLegA_2}
    componentLinearAccScale=<1 1 1>
    componentLinearVelScale=<1 1 1>
    gravityOverride {z=-980}
    physicsBodyDefinitions [{
      sphereCollisionRadius=0
      boundBone {boneName=dyn_zipperLegA_2}
      boxExtents=<5 5 5>
      constraintSetup {angularLimitsMax=<5 30 0>, angularLimitsMin=<0 -30 -30>, angularTarget=<1 0 0>}
      localJointOffset=<1 0 -0.1>
     }]
   }
   n5 = Anim Dynamics - Bone: dyn_zipperLegB_2(ComponentPose=@n4, Alpha=1, ExternalForce="0,0,0") {
    bGravityOverrideInSimSpace=true
    bUseGravityOverride=true
    lODThreshold=1
    numSolverIterationsPostUpdate=4
    numSolverIterationsPreUpdate=16
    simulationSpace=RootRelative
    alphaBoolBlend {blendOption=Linear}
    boundBone {boneName=dyn_zipperLegB_2}
    componentLinearAccScale=<1 1 1>
    componentLinearVelScale=<1 1 1>
    gravityOverride {z=-980}
    physicsBodyDefinitions [{
      sphereCollisionRadius=0
      boundBone {boneName=dyn_zipperLegB_2}
      boxExtents=<5 5 5>
      constraintSetup {angularLimitsMax=<5 30 0>, angularLimitsMin=<0 -30 -30>, angularTarget=<1 0 0>}
      localJointOffset=<1 0 -0.1>
     }]
   }
   n6 = "Constraint - const_zipper_r (PPPPPP)"!(
    ComponentPose=@n5
    Alpha=1
    ConstraintWeights_0=0.015
    ConstraintWeights_1:float=0.24
    ConstraintWeights_2:float=0.033
    ConstraintWeights_3:float=0.243
    ConstraintWeights_4:float=0.299
    ConstraintWeights_5:float=0.174
   ) -> (Pose) {
    alphaBoolBlend {blendOption=Linear}
    boneToModify {boneName=const_zipper_r}
    constraintSetup [
     {targetBone {boneName=pelvis}}
     {targetBone {boneName=clavicle_r}}
     {targetBone {boneName=spine_01}}
     {targetBone {boneName=spine_02}}
     {targetBone {boneName=spine_03}}
     {targetBone {boneName=spine_04}}
    ]
    constraintWeights [0.085,0.915,1,1,1,1]
   }
   n7 = Anim Dynamics - Bone: dyn_zipper_r(ComponentPose=@n6, Alpha=1, ExternalForce="0,0,0") {
    bGravityOverrideInSimSpace=true
    bUseGravityOverride=true
    lODThreshold=1
    numSolverIterationsPostUpdate=4
    numSolverIterationsPreUpdate=16
    simulationSpace=RootRelative
    alphaBoolBlend {blendOption=Linear}
    boundBone {boneName=dyn_zipper_r}
    componentLinearAccScale=<1 1 1>
    componentLinearVelScale=<1 1 1>
    gravityOverride {z=-980}
    physicsBodyDefinitions [{
      sphereCollisionRadius=0
      boundBone {boneName=dyn_zipper_r}
      boxExtents=<5 5 5>
      constraintSetup {angularLimitsMax=<5 30 0>, angularLimitsMin=<0 -30 -30>, angularTarget=<1 0 0>}
      localJointOffset=<4 0 -0.1>
     }]
   }
   n8 = "Constraint - const_zipper_l (PPPPPP)"!(
    ComponentPose=@n7
    Alpha=1
    ConstraintWeights_0=0.015
    ConstraintWeights_1:float=0.24
    ConstraintWeights_2:float=0.033
    ConstraintWeights_3:float=0.243
    ConstraintWeights_4:float=0.299
    ConstraintWeights_5:float=0.174
   ) -> (Pose) {
    alphaBoolBlend {blendOption=Linear}
    boneToModify {boneName=const_zipper_l}
    constraintSetup [
     {targetBone {boneName=pelvis}}
     {targetBone {boneName=clavicle_l}}
     {targetBone {boneName=spine_01}}
     {targetBone {boneName=spine_02}}
     {targetBone {boneName=spine_03}}
     {targetBone {boneName=spine_04}}
    ]
    constraintWeights [0.085,0.915,1,1,1,1]
   }
   n9 = Anim Dynamics - Bone: dyn_zipper_l(ComponentPose=@n8, Alpha=1, ExternalForce="0,0,0") {
    bGravityOverrideInSimSpace=true
    bUseGravityOverride=true
    lODThreshold=1
    numSolverIterationsPostUpdate=4
    numSolverIterationsPreUpdate=16
    simulationSpace=RootRelative
    alphaBoolBlend {blendOption=Linear}
    boundBone {boneName=dyn_zipper_l}
    componentLinearAccScale=<1 1 1>
    componentLinearVelScale=<1 1 1>
    gravityOverride {z=-980}
    physicsBodyDefinitions [{
      sphereCollisionRadius=0
      boundBone {boneName=dyn_zipper_l}
      boxExtents=<5 5 5>
      constraintSetup {angularLimitsMax=<5 30 0>, angularLimitsMin=<0 -30 -30>, angularTarget=<1 0 0>}
      localJointOffset=<4 0 -0.1>
     }]
   }
   n10 = "Constraint - dyn_lapel_l (R)"(ComponentPose=@n9, Alpha=1, ConstraintWeights_0=0.5) {
    alphaBoolBlend {blendOption=Linear}
    boneToModify {boneName=dyn_lapel_l}
    constraintSetup [{transformType=Rotation, targetBone {boneName=clavicle_l}}]
    constraintWeights [1]
   }
   n11 = "dyn_lapel_r.multiple = clavicle_r.rotateZ * 2 - Bone Driven Controller"!(ComponentPose:Component Space Pose Link Structure=@n10, Alpha:float=1) -> (Pose:Component Space Pose Link Structure) {
    is=title AnimGraphNode_BoneDrivenController
    bAffectTargetRotationY=true
    bAffectTargetRotationZ=true
    modificationMode=AddToRefPose
    multiplier=2
    rangeMax=30
    rangeMin=0
    sourceComponent=RotationZ
    alphaBoolBlend {blendOption=Linear}
    sourceBone {boneName=clavicle_r}
    targetBone {boneName=dyn_lapel_r}
   }
   n12 = Anim Dynamics - Bone: dyn_strapA_r(ComponentPose=@n11, Alpha=1, ExternalForce="0,0,0") {
    bGravityOverrideInSimSpace=true
    bUseGravityOverride=true
    lODThreshold=1
    numSolverIterationsPostUpdate=4
    numSolverIterationsPreUpdate=16
    simulationSpace=RootRelative
    alphaBoolBlend {blendOption=Linear}
    boundBone {boneName=dyn_strapA_r}
    componentLinearAccScale=<1 1 1>
    componentLinearVelScale=<1 1 1>
    gravityOverride {z=-980}
    physicsBodyDefinitions [{
      sphereCollisionRadius=0
      boundBone {boneName=dyn_strapA_r}
      boxExtents=<5 5 5>
      constraintSetup {angularLimitsMax=<5 3 0>, angularLimitsMin=<0 -3 -30>, angularTarget=<1 0 0>}
      localJointOffset=<4 0 -0.1>
     }]
   }
   n13 = "Transform (Modify) Bone - Bone: dyn_clavicle_rEND"(ComponentPose=@n12, Alpha=Get pistolDistance() -> (pistolDistance:double), Translation="2,2,0", Rotation="0,-35,0", Scale="1,1,1") {
    rotationMode=BMM_Additive
    translationMode=BMM_Additive
    alphaBoolBlend {blendOption=Linear}
    boneToModify {boneName=dyn_clavicle_rEND}
    rotation {yaw=-35}
    translation=<2 2>
   }
   n14 = Anim Dynamics - Bone: dyn_strapB_r(ComponentPose=@n13, Alpha=1, ExternalForce="0,0,0") {
    bGravityOverrideInSimSpace=true
    bUseGravityOverride=true
    lODThreshold=1
    numSolverIterationsPostUpdate=4
    numSolverIterationsPreUpdate=16
    simulationSpace=RootRelative
    alphaBoolBlend {blendOption=Linear}
    boundBone {boneName=dyn_strapB_r}
    componentLinearAccScale=<1 1 1>
    componentLinearVelScale=<1 1 1>
    gravityOverride {z=-980}
    physicsBodyDefinitions [{
      sphereCollisionRadius=0
      boundBone {boneName=dyn_strapB_r}
      boxExtents=<5 5 5>
      constraintSetup {angularLimitsMax=<5 3 0>, angularLimitsMin=<0 -3 -30>, angularTarget=<1 0 0>}
      localJointOffset=<4 0.2 -0.1>
     }]
   }
   n15 = Pose Driver - Source: neck_01!(SourcePose:Pose Link Structure=Component To Local(ComponentPose=@n14)) -> (Pose:Pose Link Structure) {
    is=title AnimGraphNode_PoseDriver
    poseAsset=&/M_MED_Rocker_Punk_Collar_RBF_PoseAsset
    onlyDriveBones [{boneName=dyn_neck_01}, {boneName=dyn_neck_02}, {boneName=dyn__head}]
    poseTargets [
     {
      bApplyCustomCurve=false
      bIsHidden=false
      distanceMethod=DefaultMethod
      drivenName=Ref
      functionType=DefaultFunction
      targetScale=0.5748
      boneTransforms [
       {targetRotation=<rot 0 23.1062 0>, targetTranslation=<11.3099 0.15 -0.0031>}
       {targetRotation=<rot 0 -5.7813 0>, targetTranslation=<4.5617 -0.0013 0.0003>}
       {targetRotation=<rot 0 -13.2022 0>, targetTranslation=<5.3902 -0.0018 0.0005>}
      ]
      customCurve {defaultValue=3.4028234663852886e+38, postInfinityExtrap=RCCE_Constant, preInfinityExtrap=RCCE_Constant, keys []}
      targetRotation=<rot 0 0 0>
     }
     {
      bApplyCustomCurve=false
      bIsHidden=false
      distanceMethod=DefaultMethod
      drivenName=LookRight
      functionType=DefaultFunction
      targetScale=0.5748
      boneTransforms [
       {targetRotation=<rot 19.0649 8.9888 -23.1981>, targetTranslation=<11.3099 0.15 -0.0031>}
       {targetRotation=<rot 3.1747 -3.832 -6.1895>, targetTranslation=<4.5617 -0.0013 0.0003>}
       {targetRotation=<rot 4.1857 -8.6638 -18.9597>, targetTranslation=<5.3902 -0.0018 0.0005>}
      ]
      customCurve {defaultValue=3.4028234663852886e+38, postInfinityExtrap=RCCE_Constant, preInfinityExtrap=RCCE_Constant, keys []}
      targetRotation=<rot 0 0 0>
     }
     {
      bApplyCustomCurve=false
      bIsHidden=false
      distanceMethod=DefaultMethod
      drivenName=LookLeft
      functionType=DefaultFunction
      targetScale=1
      boneTransforms [
       {targetRotation=<rot -24.0841 -12.8598 26.4396>, targetTranslation=<11.3049 0.1512 0.0087>}
       {targetRotation=<rot -3.861 -6.1623 5.6257>, targetTranslation=<4.5573 -0.0013 0.0021>}
       {targetRotation=<rot 5.2653 -12.4398 30.471>, targetTranslation=<5.3848 -0.0018 0.0025>}
      ]
      customCurve {defaultValue=3.4028234663852886e+38, postInfinityExtrap=RCCE_Constant, preInfinityExtrap=RCCE_Constant, keys []}
      targetRotation=<rot 0 0 0>
     }
    ]
    rBFParams {radius=21.4351}
    sourceBones [{boneName=neck_01}, {boneName=neck_02}, {boneName=head}]
   }
   n16 = Pose Driver - Source: upperarm_r!(SourcePose:Pose Link Structure=@n15) -> (Pose:Pose Link Structure) {
    is=title AnimGraphNode_PoseDriver
    driveOutput=DriveCurves
    poseAsset=&/M_MED_Rocker_Punk_RArm_RBF_PoseAsset
    poseTargets [
     {
      bApplyCustomCurve=false
      bIsHidden=false
      distanceMethod=DefaultMethod
      drivenName=Base
      functionType=DefaultFunction
      targetScale=1
      boneTransforms [{targetRotation=<rot -64.7859 -3.0892 6.0956>, targetTranslation=<-15.5111 0 0>}]
      customCurve {defaultValue=3.4028234663852886e+38, postInfinityExtrap=RCCE_Constant, preInfinityExtrap=RCCE_Constant, keys []}
      targetRotation=<rot 0 0 0>
     }
     {
      bApplyCustomCurve=false
      bIsHidden=false
      distanceMethod=DefaultMethod
      drivenName=pistolUp_rightArm_corrective
      functionType=DefaultFunction
      targetScale=1
      boneTransforms [{targetRotation=<rot 34.666 96.5719 -83.9316>, targetTranslation=<-15.5111 0 0>}]
      customCurve {defaultValue=3.4028234663852886e+38, postInfinityExtrap=RCCE_Constant, preInfinityExtrap=RCCE_Constant, keys []}
      targetRotation=<rot 0 0 0>
     }
    ]
    rBFParams {radius=46.6236}
    sourceBones [{boneName=upperarm_r}]
   }
   Output Pose(Result=@n16) {layerGroup=None}
   AnimGraphNode_SubInput_0 = Input Pose()
  ]}
 EventGraph {nodes [
   Event BlueprintUpdateAnimation()
   > Set pistolDistance(pistolDistance:double=Get Distance Between Two Sockets!(
     Component:Skeletal Mesh Component Object Reference=GetOwnerSkeletalMeshComponent()
     SocketOrBoneNameA:Name=dyn_pistolTarget
     SocketSpaceA:ERelativeTransformSpace Enum=RTS_Component
     SocketOrBoneNameB:Name=lowerarm_r
     SocketSpaceB:ERelativeTransformSpace Enum=RTS_Component
     bRemapRange:bool=true
     InRangeMin:float=125
     InRangeMax:float=147
     OutRangeMin:float=0
     OutRangeMax:float=1
    ) -> (ReturnValue:float) {is=title K2Node_CallFunction}) -> (Output_Get:double)
   TryGetPawnOwner()
  ]}
]
variables [pistolDistance {type=number}]
